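# InMaterial is an editor material in the Geometry Collection folder, set up with translucent blending and the Subsurface shading model.
# Its graph holds a single constant magenta color vector, with no connection to a material output shown.
# It uses no textures, parameters or math, renders in the before-depth-of-field translucency pass and does not respond to decals.
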Material /Engine/EditorMaterials/GeometryCollection/InMaterial
graph {
 properties {
  blendMode=BLEND_Translucent
  materialDecalResponse=MDR_None
  pixelDepthOffsetMode=PDOM_Legacy
  shadingModel=MSM_Subsurface
  translucencyPass=MTP_BeforeDOF
  usedShadingModels=Subsurface
 }
 nodes [Constant3Vector() {constant=<rgba 1 0 1 0>}]
}
