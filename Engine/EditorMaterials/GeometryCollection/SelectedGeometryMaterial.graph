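# Translucent unlit editor material that, judging by its name, highlights selected Geometry Collection pieces, with Nanite usage enabled.
# A world-aligned mask on pre-skinned position and normal drives a lerp feeding face color, detail color and a vertex-alpha product.
# Edge detection on neighboring scene base-color samples chooses detail or face color, then Fresnel and Emissive scale it.
# Forward and mobile paths substitute flat red before Emissive scaling.
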Material /Engine/EditorMaterials/GeometryCollection/SelectedGeometryMaterial
graph {
 properties {
  bUsedWithGeometryCollections=true
  bUsedWithNanite=true
  blendMode=BLEND_Translucent
  materialDecalResponse=MDR_None
  pixelDepthOffsetMode=PDOM_Legacy
  shadingModel=MSM_Unlit
  translucencyPass=MTP_BeforeDOF
  usedShadingModels=Unlit
 }
 nodes [
  n1 = Multiply(A=ScreenPosition().PixelPosition)
  n2 = Fmod(A=Floor(None=@n1), B=Constant() {r=2})
  n3 = Frac(None=@n1)
  n4 = Add(A=Multiply(A=@n2, B=OneMinus(None=@n3)), B=Multiply(A=@n3, B=OneMinus(None=@n2)))
  MaterialExpressionAbs_0 = Abs(None=Subtract(A=ComponentMask(None=@n4) {r=true}, B=ComponentMask(None=@n4) {g=true}))
  Custom!(a=@MaterialExpressionAbs_0) {
   code="return a < .1 ? 0 :1;"
   outputType=CMOT_Float1
   inputs [{inputName=a, input {expression=&self:MaterialExpressionAbs_0}}]
  }
  n5 = Constant3Vector() {constant=<rgba 1 0 0 0>}
  n6 = Constant() {r=6}
  n7 = MaterialFunctionCall() {materialFunction=/Engine/Functions/Engine_MaterialFunctions02/WorldPositionOffset/ObjectScale}
  n8 = Fresnel(ExponentIn=Add(A=@n6, B=Divide(A=@n7.Scale XYZ, B=@n6) {constB=5})) {baseReflectFraction=0.02}
  n9 = Divide(B=MaterialFunctionCall() {materialFunction=&ScreenResolution}.Buffer Resolution)
  n10 = Multiply(A=@n9, B=Constant2Vector() {r=1})
  n11 = ScreenPosition()
  n12 = Multiply(A=@n9, B=Constant2Vector() {r=-1})
  n13 = Subtract(
   A=SceneTexture(UVs=Add(A=@n11.ViewportUV, B=@n10)) {sceneTextureId=PPI_BaseColor}.Color
   B=SceneTexture(UVs=Add(A=@n11.ViewportUV, B=@n12)) {sceneTextureId=PPI_BaseColor}.Color
  )
  n14 = MaterialFunctionCall(Vector 3=Multiply(A=ComponentMask(None=@n13) {b=true, g=true, r=true}) {constB=0.5}) {materialFunction=/Engine/Functions/Engine_MaterialFunctions02/Utility/VectorLength}
  n15 = Multiply(A=@n14.V3 Length, B=@n14.V3 Length)
  n16 = Multiply(A=@n9, B=Constant2Vector() {g=1})
  n17 = Multiply(A=@n9, B=Constant2Vector() {g=-1})
  n18 = Subtract(
   A=SceneTexture(UVs=Add(A=@n11.ViewportUV, B=@n16)) {sceneTextureId=PPI_BaseColor}.Color
   B=SceneTexture(UVs=Add(A=@n11.ViewportUV, B=@n17)) {sceneTextureId=PPI_BaseColor}.Color
  )
  n19 = MaterialFunctionCall(Vector 3=Multiply(A=ComponentMask(None=@n18) {b=true, g=true, r=true}) {constB=0.5}) {materialFunction=&VectorLength}
  n20 = Multiply(A=@n19.V3 Length, B=@n19.V3 Length)
  n21 = If(
   A=Multiply(A=Add(A=@n15, B=@n20))
   B=ScalarParameter() {defaultValue=0.0009, parameterName=Outliner}
   "A > B"=Constant() {r=1}
   "A < B"=Constant()
  )
  n22 = VectorParameter() {parameterName=CheckerOpacity, channelNames {a=n/a, b=n/a, g=B, r=A}, defaultValue=<rgba 0.4 0.7 0 1>}
  n23 = TextureObjectParameter() {parameterName=Selection Mask Input, texture=/Engine/EngineMaterials/Grid}
  n24 = ScalarParameter() {defaultValue=10, parameterName=Divisions}
  n25 = VertexInterpolator(VS=PreSkinnedPosition())
  n26 = VertexInterpolator(VS=PreSkinnedNormal())
  n27 = ScalarParameter() {defaultValue=256, parameterName=Contrast}
  n28 = LinearInterpolate(
   A=@n22.A
   B=@n22.A
   Alpha=ComponentMask(None=MaterialFunctionCall(TextureObject=@n23, TextureSize=@n24, WorldPosition=@n25.PS, World Space Normal=@n26.PS, ProjectionTransitionContrast=@n27) {materialFunction=&WorldAlignedTexture}.XYZ Texture) {r=true}
  )
  n29 = Multiply(A=@n28, B=VectorParameter() {desc=Outline, parameterName=Detail Color, defaultValue=<rgba 0.4323 0.1509 0 1>}.RGB)
  n30 = OneMinus(None=@n21)
  n31 = LinearInterpolate(B=VectorParameter() {desc=Face Color, parameterName=Face Color, defaultValue=<rgba 0.7969 0.2781 0 1>}.RGB, Alpha=@n28)
  n32 = Multiply(
   A=Add(A=Multiply(A=@n21, B=@n29), B=Multiply(A=@n30, B=@n31))
   B=LinearInterpolate(Alpha=ComponentMask(None=@n8) {r=true}) {constA=0.65, constB=3.5}
  )
  Multiply(A=ScalarParameter() {defaultValue=4, parameterName=Emissive}, B=ShadingPathSwitch(Default=@n32, Deferred=@n32, Forward=@n5, Mobile=@n5))
  n33 = VertexColor()
  Multiply(A=@n28, B=@n33.A)
  Reroute(None=@n8)
  Reroute(None=@n33.A)
 ]
}
